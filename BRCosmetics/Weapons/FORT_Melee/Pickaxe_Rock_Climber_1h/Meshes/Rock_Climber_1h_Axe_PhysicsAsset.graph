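# Physics asset for the Rock Climber one-handed pickaxe, providing bodies and joints for its dangling ornament and ribbon.
# The handle bone carries two capsules and a box as the tool's main shape, and anchors both dynamic chains.
# The ornament chain runs from a master body through four links, and the ribbon from a master through two.
# Both master joints and the first ornament joint swing and twist freely, while joints farther along have angular limits, with very tight twist on the ribbon.
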PhysicsAsset /BRCosmetics/Weapons/FORT_Melee/Pickaxe_Rock_Climber_1h/Meshes/Rock_Climber_1h_Axe_PhysicsAsset
bodies {
 Melee_TwoHanded_Handle [
  {shapeType=Capsule, length=42.627, radius=1.8923, center=<-6.4034 -0.3059 36.5344>, rotation=<rot 0 -90 0>}
  {shapeType=Capsule, length=1.2192, radius=2.4461, center=<-6.9216 -0.2894 51.4807>, rotation=<rot 8.2302 0 0>}
  {shapeType=Box, extentX=29.7591, extentY=1.4913, extentZ=7.5, center=<-5.268 -0.3537 61.6225>, rotation=<rot 7.9142 0 0>}
 ]
 dyn_ornament_01 [{shapeType=Capsule, length=1.2672, radius=0.7018, center=<1.1766 -0.0001 0>, rotation=<rot 90 0 0>}]
 dyn_ornament_02 [{shapeType=Capsule, length=1.9033, radius=0.6595, center=<0.938 0 0>, rotation=<rot 90 0 0>}]
 dyn_ornament_03 [{shapeType=Capsule, length=3.5019, radius=0.5859, center=<1.7342 0 0>, rotation=<rot 90 0 0>}]
 dyn_ornament_04 [{shapeType=Box, extentX=1.0347, extentY=3.5761, extentZ=3.9472, center=<3.2844 -0.0852 -0.0813>, rotation=<rot -0.6164 -89.9986 -0.3142>}]
 dyn_ornament_mstr [{shapeType=Box, extentX=3.1759, extentY=1.6833, extentZ=2.9627, center=<-0.6002 0 0>}]
 dyn_ribbon_01 [{shapeType=Box, extentX=5.327, extentY=2.2636, extentZ=1.3219, center=<2.4605 0.3065 0.1679>}]
 dyn_ribbon_02 [{shapeType=Box, extentX=5.7982, extentY=1.7082, extentZ=5.3417, center=<2.7265 0.3206 -0.2864>}]
 dyn_ribbon_mstr [{shapeType=Box, extentX=1.01, extentY=1.01, extentZ=1.01}]
}
constraints [
 {
  bone1Name=dyn_ornament_mstr
  bone2Name=Melee_TwoHanded_Handle
  swing1LimitDegrees=45
  swing1Motion=Free
  swing2LimitDegrees=45
  swing2Motion=Free
  twistLimitDegrees=45
  twistMotion=Free
 }
 {
  bone1Name=dyn_ornament_01
  bone2Name=dyn_ornament_mstr
  swing1LimitDegrees=45
  swing1Motion=Free
  swing2LimitDegrees=45
  swing2Motion=Free
  twistLimitDegrees=45
  twistMotion=Free
 }
 {bone1Name=dyn_ornament_02, bone2Name=dyn_ornament_01, swing1LimitDegrees=90, swing2LimitDegrees=90, twistLimitDegrees=20}
 {bone1Name=dyn_ribbon_02, bone2Name=dyn_ribbon_01, swing1LimitDegrees=40, swing2LimitDegrees=70, twistLimitDegrees=5}
 {
  bone1Name=dyn_ribbon_mstr
  bone2Name=Melee_TwoHanded_Handle
  swing1LimitDegrees=45
  swing1Motion=Free
  swing2LimitDegrees=45
  swing2Motion=Free
  twistLimitDegrees=45
  twistMotion=Free
 }
 {bone1Name=dyn_ribbon_01, bone2Name=dyn_ribbon_mstr, swing1LimitDegrees=40, swing2LimitDegrees=70, twistLimitDegrees=5}
 {bone1Name=dyn_ornament_03, bone2Name=dyn_ornament_02, swing1LimitDegrees=70, swing2LimitDegrees=70, twistLimitDegrees=15}
 {bone1Name=dyn_ornament_04, bone2Name=dyn_ornament_03, swing1LimitDegrees=45, swing2LimitDegrees=45, twistLimitDegrees=45}
]
